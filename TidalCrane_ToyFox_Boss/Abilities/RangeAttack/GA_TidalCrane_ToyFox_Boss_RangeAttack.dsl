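; Toy Fox boss ranged attack ability that lobs a tick-staggered volley of projectiles at the target, then applies frustration to the boss.
; Aims at predicted player movement plus random offset, launching from a socket or the parent's location.
; Arc varies by shot index, randomness and distance, and each shot gets a spawned target-projection actor.
; Cooldown is randomized, and hotfixable scalable floats drive the tuning.
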(event EventPerformCommit
  ;[0.0104 0.15 0.07243 0.5] Randomize the cooldown.
  (Ability|CommitAbilityCooldownWithDuration (Math|Random|RandomFloatinRange (ScalableFloat|GetCachedScalableFloatValue (Variables|Hotfix|GetHotfix_CooldownMin)) (ScalableFloat|GetCachedScalableFloatValue (Variables|Hotfix|GetHotfix_CooldownMax)))))

(event EventLaunchProjectile
  ;[0.124 0.04594 0.15 0.5] LaunchProjectile: Overridden to spawn multiple projectiles.
  ; Apply frustration when projectile is fired.
  (bind _countindex (Utilities|TickNumOfProjectiles (Math|Float|Truncate (ScalableFloat|GetCachedScalableFloatValue (Variables|Hotfix|GetHotfix_NumOfProjectiles))))
    (:SpawnProjectile
      (bind _output_get (Variables|InternalData|SetTargetLocation (CallFunction|GetProjectileTargetLocation)))
      (if (Variables|Debug|GetbDebug)
        (Rendering|Debug|DrawDebugSphere _output_get 2.5e+02 12 "(R=1.000000,G=0.000000,B=0.514500,A=1.000000)" 4.0 5.0)
        (else
          _))
      (bind _outlaunchvelocity (CallFunction|GetSuggestedProjectileVelocity _output_get (CallFunction|GetProjectileSpawnLocation) (Math|Float|Clamp(Float) (Utilities|Operators|Add (ScalableFloat|GetValueatLevel (Variables|Default|GetArcParam) (Math|Conversions|ToFloat(Integer) (- _countindex 1))) (Math|Random|RandomFloatinRange (* (ScalableFloat|GetCachedScalableFloatValue (Variables|Hotfix|GetHotfix_ArcParamRandomRange)) -1.0) (ScalableFloat|GetCachedScalableFloatValue (Variables|Hotfix|GetHotfix_ArcParamRandomRange))) (Math|Float|MapRangeClamped (Math|Vector|Distance(Vector) (Transformation|GetActorLocation (Variables|InternalData|GetTargetActor)) (Transformation|GetActorLocation (Ability|GetAvatarActorfromActorInfo))) (ScalableFloat|GetCachedScalableFloatValue (Variables|Hotfix|GetHotfix_ArcRangeBasedAdjustDistMax)) (ScalableFloat|GetCachedScalableFloatValue (Variables|Hotfix|GetHotfix_ArcRangeBasedAdjustDistMin)) (ScalableFloat|GetCachedScalableFloatValue (Variables|Hotfix|GetHotfix_ArcRangeBasedAdjustMin)) (ScalableFloat|GetCachedScalableFloatValue (Variables|Hotfix|GetHotfix_ArcRangeBasedAdjustMax)))) (ScalableFloat|GetCachedScalableFloatValue (Variables|Hotfix|GetHotfix_ArcClampMin)) (ScalableFloat|GetCachedScalableFloatValue (Variables|Hotfix|GetHotfix_ArcClampMax))) (ScalableFloat|GetValueatLevel (Variables|Default|GetProjectilveGravityOverride))))
      (bind _returnvalue (Projectile|SpawnProjectile :ProjectileClass (Variables|Ability|GetProjectileClass) :RequestedBy (Ability|GetAvatarActorfromActorInfo) :SpawnLocation (CallFunction|GetProjectileSpawnLocation) :SpawnRotation "0, 0, 0" :SpawnDirection (Math|Conversions|RotationFromXVector _outlaunchvelocity) :EffectContainerSpecToApplyOnHit (Ability|MakeGameplayEffectContainerSpec "(TagName=\"Targeting.ApplicationTag.0\")") :bUseDefaultPhysics false :InitialSpeed (Math|Vector|VectorLength _outlaunchvelocity) :GravityScale (ScalableFloat|GetValueatLevel (Variables|Default|GetProjectilveGravityScale)) :ChargePercent 0.0))
      (bind _asprj_tidal_crane_medium_ranged (Utilities|Casting|CastToPrj_TidalCrane_Medium_Ranged _returnvalue))
      (Variables|Default|SetRangedProjectile _asprj_tidal_crane_medium_ranged)
      (Utilities|FlowControl|Delay 0.033)
      (bind _returnvalue_2 (Game|SpawnActorfromClass "/TidalCrane_Swarm_Boss/Blueprints/BGA_TidalCrane_Boss_ProjectileLocation.BGA_TidalCrane_Boss_ProjectileLocation_C" (Math|Transform|MakeTransform (Variables|InternalData|GetTargetLocation)) "AlwaysSpawn" "MultiplyWithRoot" 0 (Ability|GetAvatarActorfromActorInfo) 0 (Ability|GetActivatingPawn)))
      (Variables|/ProtoSwarm/Abilities/Medium/Prj_TidalCrane_Medium_Ranged.Prj_TidalCrane_Medium_Ranged_C|SetTargetProjection _returnvalue_2 (Variables|Default|GetRangedProjectile)))
    (:Complete
      (Ability|ExecuteGameplayCueOnOwner (Variables|Default|GetGC_ProjectileThrown))
      (bind _asbga_tidal_crane_boss_toy_fox (Utilities|Casting|CastToBGA_TidalCrane_Boss_ToyFox (Ability|GetAvatarActorfromActorInfo)))
      (Frustration|ApplyFrustration _asbga_tidal_crane_boss_toy_fox (ScalableFloat|GetCachedScalableFloatValue (Variables|Hotfix|GetHotfix_FrustrationAmount)) "RangeAttack"))) ; Spawn one projectile per frame.
)

(fn GetProjectileSpawnLocation
  ;[0 0.1229 0.15 0.5] GetProjectileSpawnLocation: Spawn the projectile from the socket location or from an offset location based on hotfix.
  (Utilities|FlowControl|Sequence
    (:then_0
      (if (> (ScalableFloat|GetCachedScalableFloatValue (Variables|Hotfix|GetHotfix_SpawnProjectileFromSocket)) 0)
        (bind _asbga_tidal_crane_boss_toy_fox (Utilities|Casting|CastToBGA_TidalCrane_Boss_ToyFox (Ability|GetAvatarActorfromActorInfo)))
        (bind _skeletalmeshcomponentbudgeted (Variables|/TidalCrane_Swarm_Boss/Gameplay/BGA_TidalCrane_Boss.BGA_TidalCrane_Boss_C|GetSkeletalMeshComponentBudgeted _asbga_tidal_crane_boss_toy_fox)
          (:then
            (return (Transformation|GetSocketLocation _skeletalmeshcomponentbudgeted "ProjectileSocket")))
          (:else))))
    (:then_1
      (return (|CallParentFunction)))))

(fn GetProjectileTargetLocation
  (Utilities|FlowControl|Sequence
    (:then_0
      (bind _asfort_player_pawn (Utilities|Casting|CastToFortPlayerPawn (Variables|InternalData|GetTargetActor)))
      (return (+ (FortPawn|GetMovementPredictionPosition _asfort_player_pawn (Math|Random|RandomFloatinRange (ScalableFloat|GetCachedScalableFloatValue (Variables|Hotfix|GetHotfix_PredictionDeltaTimeMin)) (ScalableFloat|GetCachedScalableFloatValue (Variables|Hotfix|GetHotfix_PredictionDeltaTimeMax)))) (* (Math|Random|RandomUnitVector) (Math|Random|RandomFloatinRange (ScalableFloat|GetCachedScalableFloatValue (Variables|Hotfix|GetHotfix_TargetLocationRandomOffsetMin)) (ScalableFloat|GetCachedScalableFloatValue (Variables|Hotfix|GetHotfix_TargetLocationRandomOffsetMax)))))))
    (:then_1
      (return (|CallParentFunction)))))
