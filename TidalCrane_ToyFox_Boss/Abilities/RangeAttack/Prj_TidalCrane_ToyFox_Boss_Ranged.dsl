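; Ranged attack projectile Blueprint for the TidalCrane ToyFox boss, built on a parent class.
; On BeginPlay it runs the parent logic, then doubles the relative scale of its NS_TidalCrane_Prj component.
; Its construction script only calls the parent, and its overlap and tick events have no logic attached.
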(event Collision|EventActorBeginOverlap
  (params (OtherActor :type Actor%20Object%20Reference)))

(event EventTick
  (params (DeltaSeconds :type Float%20%28single-precision%29)))

(event EventBeginPlay
  (|CallParentFunction)
  (Transformation|SetRelativeScale3D (Variables|Default|GetNS_TidalCrane_Prj) "2.000000,2.000000,2.000000"))

(fn ConstructionScript
  (|CallParentFunction))
